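# Niagara idle effect for the GlintRush character body, showing ten floating geometric shapes as meshes with sprites.
# Per-partition arrays set each shape's size, rotation, socket and offset, while socket sampling with sine offsets makes them follow and bob.
# GoldBlend raises spin rate and glow/gold-layer alpha; bisHandHidden zeroes all but the last two shapes.
# A disabled SplitLightning emitter spawns hue-adjustable mesh bolts from shape particles.
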NiagaraSystem /BRCosmetics/Characters/Player/Male/Medium/Bodies/M_MED_GlintRush/FX/NS_Character_GlintRush_Idle
dependencies {
 usedDataInterfaces [
  &NiagaraDataInterfaceArrayFloat3
  &NiagaraDataInterfaceArrayFloat
  &NiagaraDataInterfaceArrayInt32
  &NiagaraDataInterfaceParticleRead
  &NiagaraDataInterfaceCurve
  &NiagaraDataInterfaceSocketReader
 ]
 usedDynamicInputs [&MergeHelper_FloatSwitch, &MakeIntFromFloat]
 usedModules [
  &self:SystemSpawnScript.SystemScriptSource.SystemScriptGraph.NiagaraNodeAssignment_0.SetVariables_#1
  &self:SystemSpawnScript.SystemScriptSource.SystemScriptGraph.NiagaraNodeAssignment_2.SetVariables_#2
  &SystemState
  &self:SystemSpawnScript.SystemScriptSource.SystemScriptGraph.NiagaraNodeAssignment_1.SetVariables_#3
 ]
}
emitters {
 GeoShapes {
  emitterUpdateScript [
   &EmitterState(Life Cycle Mode=::NewEnumerator1, Loop Behavior=::NewEnumerator1, Loop Duration Mode=::NewEnumerator1, UseLoopDelay=false)
   &SpawnBurst_Instantaneous(Spawn Count=10, Use Spawn Probability=false, UseLoopCountLimit=false)
  ]
  particleSpawnScript [
   &PartitionParticles(Number Of Partitions=10)
   &InitializeParticle(
    Color=<rgba 2 1.1479 0.3542 1>
    Mesh Scale=&Multiply_VectorByFloat(
     Vector=&SelectVectorFromArray(
      Direct Array Index=$Particles.PartitionParticles.Partition
      Vector Selection Array=DataInterface() {propertyValues {FloatData [
         <0.1 0.1 0.1>
         <0.1 0.1 0.1>
         <0.1 0.1 0.1>
         <0.1 0.1 0.12>
         <0.1 0.1 0.15>
         <0.1 0.1 0.2>
         <0.1 0.1 0.1>
         <0.1 0.1 0.1>
         <0.1 0.1 0.2>
         <0.1 0.1 0.15>
        ]}}
     )
     Float=&SelectFloatFromArray(
      Direct Array Index=$Particles.PartitionParticles.Partition
      Float Selection Array=DataInterface() {propertyValues {FloatData [0.2,0.15,0.32,0.21,0.19,0.22,0.24,0.21,0.21,0.3]}}
     )
    )
    Mesh Scale Mode=::NewEnumerator4
    Sprite Size=<100 100>
    Sprite Size Mode=::NewEnumerator4
    UsePositionOffset=false
    Write Material Random=true
   )
   /SetVariables(Particles.SpriteSize=&Multiply_Vector2D_ByFloat(
     Float=350
     Vector2D=&MakeVector2D(
      X=&MakeFloatFromVector(Vector=$Particles.Scale, Channel=::NewEnumerator0)
      Y=&MakeFloatFromVector(Vector=$Particles.Scale, Channel=::NewEnumerator2)
     )
    )) {isSetParameters=true}
   &InitialMeshOrientation(
    Mesh Orientation Mode=::NewEnumerator2
    Rotation=&SelectVectorFromArray(
     Direct Array Index=$Particles.PartitionParticles.Partition
     Vector Selection Array=DataInterface() {propertyValues {FloatData [
        <0.4685 0.3063 1.6524>
        <0.3514 0.4955 1.3085>
        <0.4595 0.7838 0.3063>
        <0.6847 0.5225 0.1982>
        <0.6396 1.093 0>
        <0.2523 0.7297 0.4234>
        <0.6036 0 0.2342>
        <0 0 0>
        <0 0 0>
       ]}}
    )
    Use Rotation Vector=true
   )
   &SpriteFacingAndAlignment(
    Sprite Alignment=&ReturnMeshOrientationAxis(Mesh Orientation Axis=::ZAxis, Mesh Orientation Quat=$Particles.MeshOrientation)
    WriteAlignment=true
    WriteFacing=false
   )
  ]
  particleUpdateScript [
   &ParticleState(Kill Particles When Lifetime Has Elapsed=false, Let Infinitely Lived Particles Die When Emitter Deactivates=true)
   &UpdateMeshOrientation(
    Rotation Rate=&Lerp_Float(A=1, B=4, Alpha=$System.GoldBlend)
    Rotation Vector=&RandomRangeVector(Minimum=<0 0 0.15>, Maximum=<0 0 0.1>)
   )
   &SocketLocation(
    Interpolate Socket Sampling=true
    Orientation Sampling=::NewEnumerator1
    SamplePosOffset=true
    Sampled Position Offset=&AddVectorToPosition(
     Position=&SelectVectorFromArray(
      Direct Array Index=$Particles.PartitionParticles.Partition
      Vector Selection Array=DataInterface() {propertyValues {FloatData [<11 3 9>, <15 -5 7>, <17 0 5>, <-6 -5 -9>, <-15 5 -7>, <-18 0 -7>, <4 6 -2>, <-4 -6 0>, <17 -3 11>, <11 0 15>]}}
     )
     Vector=&MakeVector(
      X=&Sine(Normalized Angle=$Emitter.Age, Period=&RandomRangeFloat(Minimum=2, Maximum=4), Scale=1.5, Bias=0)
      Y=&Sine(Normalized Angle=$Emitter.Age, Period=&RandomRangeFloat(Minimum=2, Maximum=4), Scale=1.5, Bias=0)
      Z=&Sine(Normalized Angle=$Emitter.Age, Period=&RandomRangeFloat(Minimum=2, Maximum=4), Scale=1.5, Bias=0)
     )
    )
    Socket Index=&SelectIntFromArray(
     Direct Array Index=$Particles.PartitionParticles.Partition
     Int Selection Array=DataInterface() {propertyValues {IntData [0,0,0,1,1,1,2,3,4,4]}}
    )
    Socket Reader=$System.Socket Reader
    Socket Sampling Mode=::NewEnumerator3
    VelocitySpeedLimit=false
   )
   /SetVariables(
    Particles.GlowColor=&MakeLinearColorFromVectorAndFloat(
     "Vector (RGB)"=&MakeVectorFromLinearColorRGB(Color=$Particles.Initial.Color)
     "Float (Alpha)"=&Lerp_Float(A=0, B=1, Alpha=$System.GoldBlend)
    )
    Particles.GoldLayerColor:LinearColor=&MakeLinearColorFromVectorAndFloat("Vector (RGB)"=<1 1 1>, "Float (Alpha)"=&Lerp_Float(A=0, B=1, Alpha=$System.GoldBlend))
   ) {isSetParameters=true}
   /SetVariables(Particles.GlowDynamic:Vector4f=&MakeVector4(X=0.5, Y=&RandomRangeFloat(Minimum=0.15, Maximum=0.2), Z=1, W=1)) {isSetParameters=true}
   /SetVariables(
    Particles.Scale=&Multiply_VectorByFloat(
     Vector=$Particles.Scale
     Float=&MakeCustomFloatFromBool(
      Bool=$User.bisHandHidden
      True Float=&SelectFloatFromArray(
       Direct Array Index=$Particles.PartitionParticles.Partition
       Float Selection Array=DataInterface() {propertyValues {FloatData [0,0,0,0,0,0,0,0,1,1]}}
      )
      False Float=1
     )
    )
    Particles.SpriteSize=&Multiply_Vector2D_ByFloat(
     Float=&MakeCustomFloatFromBool(
      Bool=$User.bisHandHidden
      True Float=&SelectFloatFromArray(
       Direct Array Index=$Particles.PartitionParticles.Partition
       Float Selection Array=DataInterface() {propertyValues {FloatData [0,0,0,0,0,0,0,0,1,1]}}
      )
      False Float=1
     )
     Vector2D=$Particles.SpriteSize
    )
   ) {isSetParameters=true}
  ]
  rendererClasses [&NiagaraMeshRendererProperties, &NiagaraSpriteRendererProperties]
  renderers [
   {rendererClass=&NiagaraMeshRendererProperties, rendererIndex=0}
   {rendererClass=&NiagaraMeshRendererProperties, rendererIndex=1}
   {rendererClass=&NiagaraSpriteRendererProperties, rendererIndex=2}
  ]
 }
 SplitLightning {
  bEnabled=false
  emitterSpawnScript [
   /SetVariables(
    Emitter.BoltColor:LinearColor=&HueShiftLinearColor(
     Hue=$System.Hue
     Saturation=&Multiply_Float(A=$System.Saturation, B=0.8)
     Value=&Multiply_Float(A=$System.Value, B=2)
     Linear Color=<rgba 0.5 0.4601 0.11 1>
    )
    Emitter.GlowColor:LinearColor=&HueShiftLinearColor(Hue=$System.Hue, Saturation=$System.Saturation, Value=$System.Value, Linear Color=<rgba 0.75 0.7072 0.1509 1>)
    Emitter.InnerGlowColor:LinearColor=&HueShiftLinearColor(Hue=$System.Hue, Saturation=$System.Saturation, Value=$System.Value, Linear Color=<rgba 2 1.7834 0.809 5>)
   ) {isSetParameters=true}
   /SetVariables(
    Emitter.GlowColor_A:LinearColor=&HueShiftLinearColor(Hue=$System.Hue, Saturation=$System.Saturation, Value=$System.Value, Linear Color=<rgba 0.1 0.0685 0.0255 1>)
    Emitter.GlowColor_B:LinearColor=&HueShiftLinearColor(Hue=$System.Hue, Saturation=$System.Saturation, Value=$System.Value, Linear Color=<rgba 0.1 0.0584 0.0214 1>)
   ) {isSetParameters=true}
  ]
  emitterUpdateScript [
   &EmitterState(
    Life Cycle Mode=::NewEnumerator1
    Loop Duration=&RandomRangeFloat(Minimum=0.11, Maximum=0.99)
    UseLoopDelay=false
   )
   &SpawnParticlesFromOtherEmitter(
    Attribute Reader=DataInterface() {propertyValues {EmitterBinding {bindingMode=Other, emitterName=GeoShapes}}}
    Spawn Rate=&RandomRangeFloat(Minimum=1, Maximum=1, Randomness Mode=::NewEnumerator2, Random Seed=?)
    Spawn Rate Per Particle Cap=10
    Use Maximum Age=false
    Use Minimum Age=false
   )
  ]
  particleSpawnScript [
   &InitializeParticle(
    Color=&MultiplyLinearColor(
     Linear Color=&RandomRangeLinearColor(Minimum=<rgba 1 1 1 1>, Maximum=<rgba 1.2 1.2 1.2 2>)
     Scale Factor=&MakeVector4_FromFloat(Float=0.8)
    )
    Lifetime=&RandomRangeFloat(Minimum=0.3, Maximum=0.4)
    Mesh Scale Mode=::NewEnumerator3
    Mesh Uniform Scale=&RandomRangeFloat(Minimum=0.03, Maximum=0.05)
    UsePositionOffset=false
    Write Material Random=true
   )
   &InitialMeshOrientation(
    Mesh Orientation Mode=::NewEnumerator7
    Rotation=&RandomRangeVector(Minimum=<-0.3 -0.3 -1>, Maximum=<0.3 0.3 1>)
    Use Rotation Vector=true
   )
   &InitialMeshOrientation(Mesh Orientation Mode=::NewEnumerator7, Rotation=<0 0.25 0>, Use Rotation Vector=true) {module=InitialMeshOrientation001}
   &CameraOffset(Camera Offset Amount=10)
   &SampleParticlesFromOtherEmitter(Emitter Sampling Mode:ENiagara_EmitterLocSamplingMode=::NewEnumerator0, Mesh Scale Sampling:ENiagara_DirectReadSamplingMode=::NewEnumerator1)
  ]
  particleUpdateScript [
   &ParticleState()
   &ScaleMeshSize(Scale Factor=&VectorFromFloat(Value=&FloatFromCurve(
      FloatCurve=DataInterface() {propertyValues {Curve {keys [
          {arriveTangent=1.9115, interpMode=RCIM_Cubic, leaveTangent=1.9115, tangentMode=RCTM_User, time=0, value=0}
          {arriveTangent=0.0224, interpMode=RCIM_Cubic, leaveTangent=0.0224, tangentMode=RCTM_User, time=1, value=1}
         ]}}}
      Scale Curve=&MakeCustomFloatFromBool(Bool=$User.bIsFrontEndPreview, True Float=0.6, False Float=1)
     )))
   &ScaleMeshSize(Scale Factor=&RandomRangeVector(Minimum=<1 1 1.5>, Maximum=<1 1 2>)) {module=ScaleMeshSize001}
   &DynamicMaterialParameters(
    Index 0 Param 2=&RandomRangeFloat(Minimum=0, Maximum=0)
    Index 0 Param 4=13
    Index 1 Param 1=&FloatFromCurve(FloatCurve=DataInterface() {propertyValues {Curve {keys [<at 0 0.5>,<at 1 1>]}}})
    Index 1 Param 2=2
    Index 1 Param 3=&FloatFromCurve(FloatCurve=DataInterface() {propertyValues {Curve {keys [{interpMode=RCIM_Cubic, tangentMode=RCTM_Break, time=0.5, value=-0.5}, <at 1 0.1>]}}})
    Index 2 Param 1=&Multiply_Float(
     A=&FloatFromCurve(
      FloatCurve=DataInterface() {propertyValues {Curve {keys [{interpMode=RCIM_Cubic, leaveTangent=8.1085, tangentMode=RCTM_Break, time=0.8568, value=1}, {tangentMode=RCTM_User, time=1, value=0}]}}}
      Scale Curve=0.8
     )
     B=&Divide_Float(
      A=&MakeFloatFromVector(Vector=$Particles.Scale, Channel=::NewEnumerator0)
      B=&RandomRangeFloat(Minimum=20, Maximum=25)
     )
    )
    Index 2 Param 2=&FloatFromCurve(
     FloatCurve=DataInterface() {propertyValues {Curve {keys [
         {interpMode=RCIM_Cubic, tangentMode=RCTM_Break, time=0.7, value=0}
         {arriveTangent=-0.0859, interpMode=RCIM_Cubic, tangentMode=RCTM_Break, time=1, value=-1}
        ]}}}
     Scale Curve=20
    )
    Index 2 Param 3=&Waveform(
     Global Input Position=&RandomRangeFloat(Minimum=0, Maximum=15)
     Global Phase=$Particles.Age
     "Waveform [1]"=::NewEnumerator12
     "[1] Phase Scale"=2
     "[1] Frequency"=1
     "[1] Amplitude Min/Max"=<0 1>
    )
    Index 2 Param 4=25
    Index 3 Param 1=&FloatFromCurve(
     FloatCurve=DataInterface() {propertyValues {Curve {keys [
         {interpMode=RCIM_Cubic, leaveTangent=-0.0025, tangentMode=RCTM_Break, time=0, value=0}
         {arriveTangent=1.8528, interpMode=RCIM_Cubic, tangentMode=RCTM_Break, time=1, value=1}
        ]}}}
     Scale Curve=&RandomRangeFloat(Minimum=0, Maximum=0)
    )
    Index 3 Param 2=&FloatFromCurve(
     FloatCurve=DataInterface() {propertyValues {Curve {keys [
         {interpMode=RCIM_Cubic, leaveTangent=-0.0025, tangentMode=RCTM_Break, time=0, value=0.5}
         {arriveTangent=-0.0316, interpMode=RCIM_Cubic, tangentMode=RCTM_Break, time=1, value=-0.2}
        ]}}}
     Scale Curve=0
    )
    Index 3 Param 3=&FloatFromCurve(FloatCurve=DataInterface() {propertyValues {Curve {keys [
         {interpMode=RCIM_Cubic, leaveTangent=1.1034, tangentMode=RCTM_Break, time=0, value=-1}
         {arriveTangent=2.2399, tangentMode=RCTM_User, time=0.5, value=0.125}
        ]}}})
    Index 3 Param 4=&FloatFromCurve(FloatCurve=DataInterface() {propertyValues {Curve {keys [
         {interpMode=RCIM_Cubic, leaveTangent=1.8089, tangentMode=RCTM_Break, time=0.3, value=-0.5}
         {arriveTangent=2.0757, interpMode=RCIM_Cubic, tangentMode=RCTM_Break, time=0.9, value=1}
        ]}}})
    Module.Index0Param0Write=true
    Module.Index0Param1Write=true
    Module.Index0Param2Write=true
    Module.Index0Param3Write=true
    Module.Index1Param0Write=true
    Module.Index1Param1Write=true
    Module.Index1Param2Write=true
    Module.Index1Param3Write=true
    Module.Index2Param0Write=true
    Module.Index2Param1Write=true
    Module.Index2Param2Write=true
    Module.Index2Param3Write=true
    Module.Index3Param0Write=true
    Module.Index3Param1Write=true
    Module.Index3Param2Write=true
    Module.Index3Param3Write=true
    Write Parameter Index 1=true
    Write Parameter Index 2=true
    Write Parameter Index 3=true
   )
   &SolveForcesAndVelocity(Clamp Velocity=false, Limit Acceleration=false)
   &ScaleColor(
    Scale Alpha=&FloatFromCurve(FloatCurve=DataInterface() {propertyValues {Curve {keys [
         {tangentMode=RCTM_User, time=0, value=1}
         {arriveTangent=-4.3986, interpMode=RCIM_Cubic, leaveTangent=-4.3986, tangentMode=RCTM_User, time=0.45, value=1}
         {arriveTangent=0.1205, interpMode=RCIM_Cubic, leaveTangent=0.1205, tangentMode=RCTM_User, time=1, value=0}
        ]}}})
    Scale RGB=&Lerp_Vector(
     A=&VectorFromFloat(Value=&FloatFromCurve(FloatCurve=DataInterface() {propertyValues {Curve {keys [<at 0.2771 1>,<at 0.3786 1>]}}}, Scale Curve=20))
     B=&VectorFromFloat(Value=&RandomRangeFloat(Minimum=0, Maximum=2, Evaluation Type=::NewEnumerator1))
     Alpha=&FloatFromCurve(FloatCurve=DataInterface() {propertyValues {Curve {keys [
          {interpMode=RCIM_Cubic, leaveTangent=2.5294, tangentMode=RCTM_Break, time=0, value=0}
          {arriveTangent=-0.2166, interpMode=RCIM_Cubic, tangentMode=RCTM_Break, time=0.4, value=1}
         ]}}})
    )
    ScaleA=true
    ScaleRGB=true
   )
   &ScaleColor(ScaleA=true, ScaleRGB=true) {module=ScaleColor002}
  ]
  rendererClasses [&NiagaraMeshRendererProperties]
  renderers [
   {rendererClass=&NiagaraMeshRendererProperties, rendererIndex=0}
   {rendererClass=&NiagaraMeshRendererProperties, rendererIndex=1}
  ]
 }
}
summary {
 systemName=NS_Character_GlintRush_Idle
 emitters [
  {emitterName=GeoShapes, rendererClasses [&NiagaraMeshRendererProperties, &NiagaraSpriteRendererProperties]}
  {bEnabled=false, emitterName=SplitLightning, rendererClasses [&NiagaraMeshRendererProperties]}
 ]
 userVariables [
  User.bIsFrontEndPreview {type=NiagaraBool}
  User.bisHandHidden {type=NiagaraBool}
  User.GoldBlend {type=NiagaraFloat, defaultValue=0, description=""}
  User.VariantIndex {type=NiagaraFloat, defaultValue=0, description=""}
 ]
}
