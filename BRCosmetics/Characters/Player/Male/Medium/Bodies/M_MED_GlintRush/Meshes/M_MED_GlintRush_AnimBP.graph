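# Anim Blueprint for the GlintRush male body cosmetic, adding secondary motion over the pose copied from its attached parent mesh.
# A Control Rig takes crouch, strafe, parachute, slide and emote inputs, then Anim Dynamics springs jacket, chain, guns and pouch, with front jacket nodes at zero alpha.
# The event graph derives those crouch and movement values and lowers chain alpha when crouching or rail-grinding.
# Its computed coat alpha, gravity override and rigid-body settings go unread by the AnimGraph.
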AnimBlueprint /BRCosmetics/Characters/Player/Male/Medium/Bodies/M_MED_GlintRush/Meshes/M_MED_GlintRush_AnimBP
target_skeleton=&/M_MED_GlintRush_Skeleton
graphs [
 AnimGraph {nodes [
   n1 = Control Rig!(
    Source=Copy Pose From Mesh() {bUseAttachedParent=true}
    Alpha
    IS Crouching:double=Get Crouch() -> (Crouch:double)
    moveLeft:double="Get moveLeft "() -> ("moveLeft ":double)
    moveRight:double="Get moveRight "() -> ("moveRight ":double)
    ParachuteOpened:bool=Get bIsParachuteOpened()
    BaseballSliding:bool=Get bIsBaseballSliding()
    PlayingEmote:bool=Get bIsPlayingEmote()
    IS CrouchWalk:double=Get Crouch Walk() -> (Crouch Walk:double)
   ) -> (Pose) {controlRigAssetReference {blueprintRigClass=&/M_MED_GlintRush_CtrlRig.M_MED_GlintRush_CtrlRig_C}}
   n2 = Anim Dynamics - Bone: dyn_Jacket_front_l_01!(ComponentPose=Local To Component(LocalPose=@n1), Alpha=0, GravityOverride:Vector="0,0,-980", ExternalForce="0,0,0") -> (Pose) {
    angularDampingOverride=1
    angularSpringConstant=8000
    bAngularSpring=true
    bOverrideAngularDamping=true
    bUseGravityOverride=true
    gravityScale=0.35
    simulationSpace=RootRelative
    boundBone {boneName=dyn_Jacket_front_l_01}
    componentAppliedLinearAccClamp=<3000 3000 3000>
    componentLinearAccScale=<0.5 0.5 0.5>
    componentLinearVelScale=<0.5 0.5 0.5>
    gravityOverride {z=-980}
    physicsBodyDefinitions [{
      boundBone {boneName=dyn_Jacket_front_l_01}
      boxExtents=<20 30 30>
      constraintSetup {angularLimitsMax=<0 2 0>, angularLimitsMin=<0 -2 -40>, angularTarget=<1 0 0>}
      localJointOffset=<4 0 0>
     }]
    simSpaceSettings {simSpaceAngularAlpha=0.4}
   }
   n3 = Anim Dynamics - Bone: dyn_Jacket_side_l!(ComponentPose=@n2, Alpha=1, GravityOverride:Vector="0,0,-980", ExternalForce="0,0,0") -> (Pose) {
    angularDampingOverride=1
    angularSpringConstant=8000
    bAngularSpring=true
    bOverrideAngularDamping=true
    bUseGravityOverride=true
    gravityScale=0.35
    simulationSpace=RootRelative
    boundBone {boneName=dyn_Jacket_side_l}
    componentAppliedLinearAccClamp=<3000 3000 3000>
    componentLinearAccScale=<0.5 0.5 0.5>
    componentLinearVelScale=<0.5 0.5 0.5>
    gravityOverride {z=-980}
    physicsBodyDefinitions [{
      boundBone {boneName=dyn_Jacket_side_l}
      boxExtents=<20 30 30>
      constraintSetup {angularLimitsMax=<0 2 0>, angularLimitsMin=<0 -2 -40>, angularTarget=<1 0 0>}
      localJointOffset=<4 0 0>
     }]
    simSpaceSettings {simSpaceAngularAlpha=0.4}
   }
   n4 = Anim Dynamics - Bone: dyn_Jacket_back_l!(ComponentPose=@n3, Alpha=1, GravityOverride:Vector="0,0,-980", ExternalForce="0,0,0") -> (Pose) {
    angularDampingOverride=1
    angularSpringConstant=8000
    bAngularSpring=true
    bOverrideAngularDamping=true
    bUseGravityOverride=true
    gravityScale=0.35
    simulationSpace=RootRelative
    boundBone {boneName=dyn_Jacket_back_l}
    componentAppliedLinearAccClamp=<3000 3000 3000>
    componentLinearAccScale=<0.5 0.5 0.5>
    componentLinearVelScale=<0.5 0.5 0.5>
    gravityOverride {z=-980}
    physicsBodyDefinitions [{
      boundBone {boneName=dyn_Jacket_back_l}
      boxExtents=<20 30 30>
      constraintSetup {angularLimitsMax=<0 2 0>, angularLimitsMin=<0 -2 -40>, angularTarget=<1 0 0>}
      localJointOffset=<4 0 0>
     }]
    simSpaceSettings {simSpaceAngularAlpha=0.4}
   }
   n5 = Anim Dynamics - Bone: dyn_Jacket_front_r!(ComponentPose=@n4, Alpha=0, GravityOverride:Vector="0,0,-980", ExternalForce="0,0,0") -> (Pose) {
    angularDampingOverride=1
    angularSpringConstant=8000
    bAngularSpring=true
    bOverrideAngularDamping=true
    bUseGravityOverride=true
    gravityScale=0.35
    simulationSpace=RootRelative
    boundBone {boneName=dyn_Jacket_front_r}
    componentAppliedLinearAccClamp=<3000 3000 3000>
    componentLinearAccScale=<0.5 0.5 0.5>
    componentLinearVelScale=<0.5 0.5 0.5>
    gravityOverride {z=-980}
    physicsBodyDefinitions [{
      boundBone {boneName=dyn_Jacket_front_r}
      boxExtents=<20 30 30>
      constraintSetup {angularLimitsMax=<0 2 0>, angularLimitsMin=<0 -2 -40>, angularTarget=<1 0 0>}
      localJointOffset=<4 0 0>
     }]
    simSpaceSettings {simSpaceAngularAlpha=0.4}
   }
   n6 = Anim Dynamics - Bone: dyn_Jacket_side_r!(ComponentPose=@n5, Alpha=1, GravityOverride:Vector="0,0,-980", ExternalForce="0,0,0") -> (Pose) {
    angularDampingOverride=1
    angularSpringConstant=8000
    bAngularSpring=true
    bOverrideAngularDamping=true
    bUseGravityOverride=true
    gravityScale=0.35
    simulationSpace=RootRelative
    boundBone {boneName=dyn_Jacket_side_r}
    componentAppliedLinearAccClamp=<3000 3000 3000>
    componentLinearAccScale=<0.5 0.5 0.5>
    componentLinearVelScale=<0.5 0.5 0.5>
    gravityOverride {z=-980}
    physicsBodyDefinitions [{
      boundBone {boneName=dyn_Jacket_side_r}
      boxExtents=<20 30 30>
      constraintSetup {angularLimitsMax=<0 2 0>, angularLimitsMin=<0 -2 -40>, angularTarget=<1 0 0>}
      localJointOffset=<4 0 0>
     }]
    simSpaceSettings {simSpaceAngularAlpha=0.4}
   }
   n7 = Anim Dynamics - Bone: dyn_Jacket_back_r!(ComponentPose=@n6, Alpha=1, GravityOverride:Vector="0,0,-980", ExternalForce="0,0,0") -> (Pose) {
    angularDampingOverride=1
    angularSpringConstant=8000
    bAngularSpring=true
    bOverrideAngularDamping=true
    bUseGravityOverride=true
    gravityScale=0.35
    simulationSpace=RootRelative
    boundBone {boneName=dyn_Jacket_back_r}
    componentAppliedLinearAccClamp=<3000 3000 3000>
    componentLinearAccScale=<0.5 0.5 0.5>
    componentLinearVelScale=<0.5 0.5 0.5>
    gravityOverride {z=-980}
    physicsBodyDefinitions [{
      boundBone {boneName=dyn_Jacket_back_r}
      boxExtents=<20 30 30>
      constraintSetup {angularLimitsMax=<0 2 0>, angularLimitsMin=<0 -2 -40>, angularTarget=<1 0 0>}
      localJointOffset=<4 0 0>
     }]
    simSpaceSettings {simSpaceAngularAlpha=0.4}
   }
   n8 = Anim Dynamics - Bone: dyn_chain!(ComponentPose=@n7, Alpha=Get DynamicAlpha() -> (DynamicAlpha:double), GravityOverride:Vector="0,0,-980", ExternalForce="0,0,0") -> (Pose) {
    angularDampingOverride=1
    angularSpringConstant=16000
    bAngularSpring=true
    bGravityOverrideInSimSpace=true
    bOverrideAngularDamping=true
    bUseGravityOverride=true
    gravityScale=0.35
    simulationSpace=RootRelative
    alphaBoolBlend {blendOption=Linear}
    boundBone {boneName=dyn_chain}
    componentAppliedLinearAccClamp=<2000 2000 2000>
    componentLinearAccScale=<0.5 0.5 0.5>
    componentLinearVelScale=<0.5 0.5 0.5>
    gravityOverride {z=-980}
    physicsBodyDefinitions [{
      sphereCollisionRadius=0
      boundBone {boneName=dyn_chain}
      boxExtents=<15 20 20>
      constraintSetup {angularLimitsMax=<0 5 0>, angularLimitsMin=<0 -5 -50>, angularTarget=<0 0 0>}
      localJointOffset=<8 0 0>
     }]
    simSpaceSettings {simSpaceAngularAlpha=0.35}
   }
   n9 = Anim Dynamics - Bone: dyn_gun_l!(ComponentPose=@n8, Alpha=1, GravityOverride:Vector="0,0,-980", ExternalForce="0,0,0") -> (Pose) {
    angularDampingOverride=1
    angularSpringConstant=1500
    bAngularSpring=true
    bOverrideAngularDamping=true
    bUseGravityOverride=true
    gravityScale=0.35
    simulationSpace=RootRelative
    alphaScaleBiasClamp {scale=0.6}
    boundBone {boneName=dyn_gun_l}
    componentAppliedLinearAccClamp=<2000 2000 2000>
    componentLinearAccScale=<0.2 0.2 0.2>
    componentLinearVelScale=<0.2 0.2 0.2>
    gravityOverride {z=-980}
    physicsBodyDefinitions [{
      boundBone {boneName=dyn_gun_l}
      boxExtents=<20 30 30>
      constraintSetup {angularLimitsMax=<0 5 0>, angularLimitsMin=<0 -5 -5>, angularTarget=<1 0 0>}
      localJointOffset=<2 0 0>
     }]
    simSpaceSettings {simSpaceAngularAlpha=0.2}
   }
   n10 = Anim Dynamics - Bone: dyn_gun_r!(ComponentPose=@n9, Alpha=1, GravityOverride:Vector="0,0,-980", ExternalForce="0,0,0") -> (Pose) {
    angularDampingOverride=1
    angularSpringConstant=1500
    bAngularSpring=true
    bOverrideAngularDamping=true
    bUseGravityOverride=true
    gravityScale=0.35
    simulationSpace=RootRelative
    alphaScaleBiasClamp {scale=0.6}
    boundBone {boneName=dyn_gun_r}
    componentAppliedLinearAccClamp=<2000 2000 2000>
    componentLinearAccScale=<0.2 0.2 0.2>
    componentLinearVelScale=<0.2 0.2 0.2>
    gravityOverride {z=-980}
    physicsBodyDefinitions [{
      boundBone {boneName=dyn_gun_r}
      boxExtents=<20 30 30>
      constraintSetup {angularLimitsMax=<0 5 0>, angularLimitsMin=<0 -5 -5>, angularTarget=<1 0 0>}
      localJointOffset=<-2 0 0>
     }]
    simSpaceSettings {simSpaceAngularAlpha=0.2}
   }
   n11 = Anim Dynamics - Bone: dyn_pouch(ComponentPose=@n10, Alpha=1, ExternalForce="0,0,0") {
    angularDampingOverride=1
    angularSpringConstant=8000
    bAngularSpring=true
    bOverrideAngularDamping=true
    bUseGravityOverride=true
    gravityScale=0.35
    simulationSpace=RootRelative
    boundBone {boneName=dyn_pouch}
    componentLinearAccScale=<1 1 1>
    componentLinearVelScale=<1 1 1>
    gravityOverride {z=-980}
    physicsBodyDefinitions [{
      boundBone {boneName=dyn_pouch}
      boxExtents=<20 20 20>
      constraintSetup {angularLimitsMax=<0 5 0>, angularLimitsMin=<0 -5 -5>, angularTarget=<1 0 0>}
      localJointOffset=<1.5 0 0>
     }]
    simSpaceSettings {simSpaceAngularAlpha=0.1}
   }
   Output Pose(Result=Component To Local(ComponentPose=@n11)) {layerGroup=None}
  ]}
 EventGraph {nodes [
   Event BlueprintUpdateAnimation()
   > n1 = Sequence()
   > n2 = "Gravity Override State Function (Parameter Asset) (Nativized)"(
    CharacterPartAnimInstance=Self-Reference()
    GameTimeInSeconds=GetGameTimeInSeconds()
    Parameters=&/M_MED_GlintRush_Parameters
   )
   > Set GravityOverride(GravityOverride=@n2)
   > Set RBAN Settings Out(RBAN Settings Out=@n2.OutRigidBodyAnimNodeParameters)
   n3 = NOT Boolean(A=OR Boolean!(A=Get bIsSliding(), B=Get bIsBaseballSliding(), C:bool=Get bIsMoving2D()) -> (ReturnValue))
   n4 = Get bIsCrouching()
   > Set Crouch(Crouch:double="To Float (Boolean)"(InBool=AND Boolean(A=@n4, B=@n3))) -> (Output_Get:double)
   n5 = Get bIsMoving2D()
   n6 = Get bIsCrouching()
   > Set Crouch Walk(Crouch Walk:double="To Float (Boolean)"(InBool=AND Boolean(A=@n6, B=@n5))) -> (Output_Get:double)
   > n7 = Branch(Condition=OR Boolean!(A=Get bIsCrouching(), B=Get bIsOnGrindRail()) -> (ReturnValue))
   > Set DynamicAlpha(DynamicAlpha:double=0.4) -> (Output_Get:double)
   > Set CoatAlpha(CoatAlpha:double=0) -> (Output_Get:double)
   Set DynamicAlpha(execute=@n7.else, DynamicAlpha=1)
   > Set CoatAlpha(CoatAlpha=1)
   n8 = PlayerMovement(execute=@n1.then_1, AnimInstance=Self-Reference(), XDotProductCoefficient=1, YDotProductCoefficient=1, ZDotProductCoefficient=1)
   n9 = BreakVector(InVec=MakeVector(X=@n8, Y=@n8.YMovement))
   > "Set moveLeft "("moveLeft ":double="Absolute (Float)"(A="Clamp (Float)"(Value=@n9, Min=-1, Max=0))) -> (Output_Get:double)
   > "Set moveRight "("moveRight ":double="Clamp (Float)"(Value=@n9, Min=0, Max=1)) -> (Output_Get:double)
   > Set moveForward(moveForward:double="Clamp (Float)"(Value=@n9.Y, Min=0, Max=1)) -> (Output_Get:double)
   > Set moveBackward(moveBackward:double="Absolute (Float)"(A="Clamp (Float)"(Value=@n9.Y, Min=-1, Max=0))) -> (Output_Get:double)
  ]}
]
variables [
 CoatAlpha {type=number}
 Crouch {type=number}
 Crouch Walk {type=number}
 DynamicAlpha {type=number}
 GravityOverride {type=Vector}
 moveBackward {type=number}
 moveForward {type=number}
 "moveLeft " {type=number}
 "moveRight " {type=number}
 RBAN Settings Out {type=FortRigidBodyAnimNodeParameters}
]
